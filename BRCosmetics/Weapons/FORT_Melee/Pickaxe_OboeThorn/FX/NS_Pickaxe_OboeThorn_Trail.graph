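# Niagara swing trail for the OboeThorn pickaxe, built from two anim-trail ribbon emitters on separate bone or socket indices.
# Base draws a wide pale ribbon, while Red adds a narrower red one that tapers and has a camera offset.
# Each has two ribbon renderers and sets base or reversed ribbon flags from the alternating SwingSectionIndex.
# Alpha fades by curve, scaled by system age and swing section, and color is tinted by time of day.
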NiagaraSystem /BRCosmetics/Weapons/FORT_Melee/Pickaxe_OboeThorn/FX/NS_Pickaxe_OboeThorn_Trail
dependencies {
 usedDataInterfaces [&NiagaraDataInterfaceCurve, &NiagaraDataInterfaceSocketReader]
 usedDynamicInputs [&FNTimeOfDayColor]
 usedModules [
  &SystemState
  &self:SystemSpawnScript.SystemScriptSource.SystemScriptGraph.NiagaraNodeAssignment_0.SetVariables_#1
 ]
}
emitters {
 Base {
  emitterUpdateScript [
   &EmitterState()
   &FN_SpawnAnimTrail(Spawn Interval=0)
   /SetVariables(
    Emitter.EnableBaseRibbon:NiagaraBool=&Boolean_OrOperation(
     Bool1=&SetBoolByIntComparison(A=$User.SwingSectionIndex, B=0, Comparison Type=::NewEnumerator4)
     Bool2=&SetBoolByIntComparison(A=$User.SwingSectionIndex, B=2, Comparison Type=::NewEnumerator4)
    )
    Emitter.EnableReversedRibbon:NiagaraBool=&Boolean_OrOperation(
     Bool1=&SetBoolByIntComparison(A=$User.SwingSectionIndex, B=1, Comparison Type=::NewEnumerator4)
     Bool2=&SetBoolByIntComparison(A=$User.SwingSectionIndex, B=3, Comparison Type=::NewEnumerator4)
    )
   ) {isSetParameters=true}
  ]
  particleSpawnScript [
   &FN_InitializeAnimTrailBoneIndecies(First Bone or Socket Index=4, Second Bone or Socket Index=4)
   &FN_InitializeAnimTrail(AnimTrailWidth=20, Color=<rgba 0.855 0.85 1 1>, Lifetime=0.2, Position Offset=<0 0 5>)
  ]
  particleUpdateScript [
   &FN_UpdateAnimTrail()
   &ScaleColor(
    Scale Alpha=&Multiply_Float(
     A=&FloatFromCurve(FloatCurve=DataInterface() {propertyValues {Curve {keys [
          {arriveTangent=-2.003, interpMode=RCIM_Cubic, leaveTangent=-2.003, tangentMode=RCTM_User, time=0.3, value=1}
          {interpMode=RCIM_Cubic, tangentMode=RCTM_User, time=1, value=0}
         ]}}})
     B=&RemapRange(
      Input Value To Remap=$Engine.System.Age
      Input Min=0
      Input Max=1
      Output Min=&MergeHelper_FloatSwitch(0=0.25, 1=0.25, 2=0.2, 3=0.25, Index=$User.SwingSectionIndex, Num=::NewEnumerator2)
      Output Max=&MergeHelper_FloatSwitch(0=0.35, 1=0.35, 2=0.2, 3=0.35, Index=$User.SwingSectionIndex, Num=::NewEnumerator2)
     )
    )
    Scale RGB=&MakeVectorFromLinearColorRGB(Color=$System.ToD)
    ScaleA=true
    ScaleRGB=true
   )
  ]
  rendererClasses [&NiagaraRibbonRendererProperties]
  renderers [
   {rendererClass=&NiagaraRibbonRendererProperties}
   {rendererClass=&NiagaraRibbonRendererProperties, rendererIndex=1}
  ]
 }
 Red {
  emitterUpdateScript [
   &EmitterState()
   &FN_SpawnAnimTrail(Spawn Interval=0)
   /SetVariables(
    Emitter.EnableBaseRibbon:NiagaraBool=&Boolean_OrOperation(
     Bool1=&SetBoolByIntComparison(A=$User.SwingSectionIndex, B=0, Comparison Type=::NewEnumerator4)
     Bool2=&SetBoolByIntComparison(A=$User.SwingSectionIndex, B=2, Comparison Type=::NewEnumerator4)
    )
    Emitter.EnableReversedRibbon:NiagaraBool=&Boolean_OrOperation(
     Bool1=&SetBoolByIntComparison(A=$User.SwingSectionIndex, B=1, Comparison Type=::NewEnumerator4)
     Bool2=&SetBoolByIntComparison(A=$User.SwingSectionIndex, B=3, Comparison Type=::NewEnumerator4)
    )
   ) {isSetParameters=true}
  ]
  particleSpawnScript [
   &FN_InitializeAnimTrailBoneIndecies(First Bone or Socket Index=5, Second Bone or Socket Index=5)
   &FN_InitializeAnimTrail(AnimTrailWidth=15, Color=<rgba 0.6376 0.0395 0.0437 5>, Lifetime=0.2)
  ]
  particleUpdateScript [
   &FN_UpdateAnimTrail()
   &ScaleColor(
    Scale Alpha=&Multiply_Float(
     A=&FloatFromCurve(FloatCurve=DataInterface() {propertyValues {Curve {keys [
          {arriveTangent=-2.003, interpMode=RCIM_Cubic, leaveTangent=-2.003, tangentMode=RCTM_User, time=0.3, value=1}
          {interpMode=RCIM_Cubic, tangentMode=RCTM_User, time=1, value=0}
         ]}}})
     B=&RemapRange(
      Input Value To Remap=$Engine.System.Age
      Input Min=0
      Input Max=1
      Output Min=&MergeHelper_FloatSwitch(0=0.25, 1=0.25, 2=0.2, 3=0.25, Index=$User.SwingSectionIndex, Num=::NewEnumerator2)
      Output Max=&MergeHelper_FloatSwitch(0=0.35, 1=0.35, 2=0.2, 3=0.35, Index=$User.SwingSectionIndex, Num=::NewEnumerator2)
     )
    )
    Scale RGB=&MakeVectorFromLinearColorRGB(Color=$System.ToD)
    ScaleA=true
    ScaleRGB=true
   )
   &ScaleRibbonWidth(Ribbon Width Scale=&FloatFromCurve(FloatCurve=DataInterface() {propertyValues {Curve {keys [<at 0 1>,<at 1 0>]}}}))
   &CameraOffset(Camera Offset Amount=1)
  ]
  rendererClasses [&NiagaraRibbonRendererProperties]
  renderers [
   {rendererClass=&NiagaraRibbonRendererProperties}
   {rendererClass=&NiagaraRibbonRendererProperties, rendererIndex=1}
  ]
 }
}
summary {
 systemName=NS_Pickaxe_OboeThorn_Trail
 emitters [
  {emitterName=Red, rendererClasses [&NiagaraRibbonRendererProperties]}
  {emitterName=Base, rendererClasses [&NiagaraRibbonRendererProperties]}
 ]
 userVariables [
  User.bisSwinging {type=NiagaraBool}
  User.RibbonIDOffset {type=NiagaraInt32}
  User.SwingSectionIndex {type=NiagaraInt32}
 ]
}
