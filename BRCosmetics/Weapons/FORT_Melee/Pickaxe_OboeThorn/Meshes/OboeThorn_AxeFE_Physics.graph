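# Physics asset for the OboeThorn melee pickaxe, defining collision and joints for a whip and a handle tuft that hang off the two-handed handle.
# Every bone uses one capsule, with larger ones on the handle and the whip's end bone.
# Twelve whip bones and four tuft bones form two constraint chains rooted at the handle.
# The tuft root is locked to the handle, while whip swing limits widen toward the tip with zero-degree twist limits.
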PhysicsAsset /BRCosmetics/Weapons/FORT_Melee/Pickaxe_OboeThorn/Meshes/OboeThorn_AxeFE_Physics
bodies {
 Melee_TwoHanded_Handle [{
   shapeType=Capsule
   length=15.8913
   radius=17.3999
   shapeName=Melee_TwoHanded_Handle_capsule
   center=<0 0 -13.1432>
   rotation=<rot 0 -104.296 -24.0703>
  }]
 dyn_handleTuft_01 [{shapeType=Capsule, length=0.8958, radius=3.2771, shapeName=dyn_handleTuft_01_capsule, center=<0.5 0 0>, rotation=<rot 90 0 0>}]
 dyn_handleTuft_02 [{shapeType=Capsule, length=4.0298, radius=3.6915, shapeName=dyn_handleTuft_01_capsule, center=<0.5 0 0>, rotation=<rot 90 0 0>}]
 dyn_handleTuft_03 [{shapeType=Capsule, length=2.973, radius=3.2771, shapeName=dyn_handleTuft_01_capsule, center=<0.5 0 0>, rotation=<rot 90 0 0>}]
 dyn_handleTuft_04 [{shapeType=Capsule, length=3.4366, radius=1.8863, shapeName=dyn_handleTuft_01_capsule, center=<0.5 0 0>, rotation=<rot 90 0 0>}]
 whip_skin_01 [{shapeType=Capsule, length=2, radius=1.5, center=<0.05 0 0>, rotation=<rot 90 0 0>}]
 whip_skin_02 [{shapeType=Capsule, length=2, radius=1.5, center=<0.05 0 0>, rotation=<rot 90 0 0>}]
 whip_skin_03 [{shapeType=Capsule, length=2, radius=2, center=<0.05 0 0>, rotation=<rot 90 0 0>}]
 whip_skin_04 [{shapeType=Capsule, length=2, radius=2, center=<0.1 0 0>, rotation=<rot 90 0 0>}]
 whip_skin_05 [{shapeType=Capsule, length=2, radius=2, center=<0.1 0 0>, rotation=<rot 90 0 0>}]
 whip_skin_06 [{shapeType=Capsule, length=2, radius=2, center=<0.1 0 0>, rotation=<rot 90 0 0>}]
 whip_skin_07 [{shapeType=Capsule, length=2, radius=2, center=<0.1 0 0>, rotation=<rot 90 0 0>}]
 whip_skin_08 [{shapeType=Capsule, length=2, radius=2, center=<0.1 0 0>, rotation=<rot 90 0 0>}]
 whip_skin_09 [{shapeType=Capsule, length=2, radius=2, center=<0.1 0 0>, rotation=<rot 90 0 0>}]
 whip_skin_10 [{shapeType=Capsule, length=2, radius=2, center=<0.1 0 0>, rotation=<rot 90 0 0>}]
 whip_skin_11 [{shapeType=Capsule, length=2, radius=2, center=<0.1 0 0>, rotation=<rot 90 0 0>}]
 whip_skin_12 [{shapeType=Capsule, length=7.8044, radius=13.2521, center=<10.649 0 1.3021>, rotation=<rot 90 0 0>}]
}
constraints [
 {bone1Name=whip_skin_02, bone2Name=whip_skin_01, swing1LimitDegrees=10, swing2LimitDegrees=15, twistLimitDegrees=0}
 {bone1Name=whip_skin_03, bone2Name=whip_skin_02, swing1LimitDegrees=10, swing2LimitDegrees=15, twistLimitDegrees=0}
 {bone1Name=whip_skin_04, bone2Name=whip_skin_03, swing1LimitDegrees=10, swing2LimitDegrees=15, twistLimitDegrees=0}
 {bone1Name=whip_skin_05, bone2Name=whip_skin_04, swing1LimitDegrees=10, swing2LimitDegrees=15, twistLimitDegrees=0}
 {bone1Name=whip_skin_06, bone2Name=whip_skin_05, swing1LimitDegrees=10, swing2LimitDegrees=15, twistLimitDegrees=0}
 {bone1Name=whip_skin_07, bone2Name=whip_skin_06, swing1LimitDegrees=10, swing2LimitDegrees=15, twistLimitDegrees=0}
 {bone1Name=whip_skin_08, bone2Name=whip_skin_07, swing1LimitDegrees=10, swing2LimitDegrees=15, twistLimitDegrees=0}
 {bone1Name=whip_skin_09, bone2Name=whip_skin_08, swing1LimitDegrees=10, swing2LimitDegrees=15, twistLimitDegrees=0}
 {bone1Name=whip_skin_10, bone2Name=whip_skin_09, swing1LimitDegrees=10, swing2LimitDegrees=20, twistLimitDegrees=0}
 {bone1Name=whip_skin_11, bone2Name=whip_skin_10, swing1LimitDegrees=10, swing2LimitDegrees=20, twistLimitDegrees=0}
 {bone1Name=whip_skin_12, bone2Name=whip_skin_11, swing1LimitDegrees=25, swing2LimitDegrees=20, twistLimitDegrees=0, twistMotion=Locked}
 {bone1Name=dyn_handleTuft_02, bone2Name=dyn_handleTuft_01, swing1LimitDegrees=20, swing2LimitDegrees=20, twistLimitDegrees=0}
 {bone1Name=dyn_handleTuft_03, bone2Name=dyn_handleTuft_02, swing1LimitDegrees=20, swing2LimitDegrees=20, twistLimitDegrees=0}
 {bone1Name=dyn_handleTuft_04, bone2Name=dyn_handleTuft_03, swing1LimitDegrees=20, swing2LimitDegrees=20, twistLimitDegrees=0}
 {
  bone1Name=dyn_handleTuft_01
  bone2Name=Melee_TwoHanded_Handle
  swing1LimitDegrees=45
  swing1Motion=Locked
  swing2LimitDegrees=45
  swing2Motion=Locked
  twistLimitDegrees=45
  twistMotion=Locked
 }
 {bone1Name=whip_skin_01, bone2Name=Melee_TwoHanded_Handle, swing1LimitDegrees=5, swing2LimitDegrees=10, twistLimitDegrees=0}
]
